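; Scrap heap engine piston that loops in the environment, animated by material world position offset with timed sound, sparks and camera shake.
; After a start delay, BeginPlay runs a cycle and sets a looping timer on PistonInit to repeat it.
; Clients restamp game time and loop duration into custom primitive data, then fire fixed-delay impact effects at a socket, shaking players inside Box.
; The authority branch is only a placeholder delay.
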(event EventBeginPlay
  (Utilities|FlowControl|Delay (Variables|Default|GetInitialStartDelay))
  (Utilities|Time|SetTimerbyEvent (delegate Custom|PistonInit) (Variables|Default|GetLoopDuration) true)
  (Utilities|FlowControl|Sequence
    (:then_0
      (Networking|SwitchHasAuthority
        (:Authority
          ;[1 1 1] if we need player knockback (probably not, i wouldnt ;p) add it here
          (Utilities|FlowControl|Delay 4.0) ; half of loop time
        )) ; server functionality, impulses etc
    )
    (:then_1
      (Networking|SwitchHasAuthority
        (:Remote
          (Utilities|FlowControl|Sequence
            (:then_0
              ;[1 1 1] Start moving from bottom /  set mesh wpo material to 0 here to reset
              (CallFunction|StartMovingSound) ;  one shot sound might work for this to make sure it keeps in sync on loop with lag
              (CallFunction|ResetMeshWPO) ; material movement resets here to keep in sync
            )
            (:then_1
              ;[1 1 1] Impact at top?
              (Utilities|FlowControl|Delay 5.0) ; half of loop time when at apex
              (Utilities|FlowControl|Delay 1.58) ; half of loop time when at apex
              (CallFunction|StartImpactSound) ; might not need this later, useful for now
              (CallFunction|SpawnParticleFX)
              (CallFunction|CameraShakeonImpact))))) ; client functionality, fx, sounds, camera shake
    )))

(event Custom|PistonInit
  (Utilities|FlowControl|Sequence
    (:then_0
      (Networking|SwitchHasAuthority
        (:Authority
          ;[1 1 1] if we need player knockback (probably not, i wouldnt ;p) add it here
          (Utilities|FlowControl|Delay 4.0) ; half of loop time
        )) ; server functionality, impulses etc
    )
    (:then_1
      (Networking|SwitchHasAuthority
        (:Remote
          (Utilities|FlowControl|Sequence
            (:then_0
              ;[1 1 1] Start moving from bottom /  set mesh wpo material to 0 here to reset
              (CallFunction|StartMovingSound) ;  one shot sound might work for this to make sure it keeps in sync on loop with lag
              (CallFunction|ResetMeshWPO) ; material movement resets here to keep in sync
            )
            (:then_1
              ;[1 1 1] Impact at top?
              (Utilities|FlowControl|Delay 5.0) ; half of loop time when at apex
              (Utilities|FlowControl|Delay 1.58) ; half of loop time when at apex
              (CallFunction|StartImpactSound) ; might not need this later, useful for now
              (CallFunction|SpawnParticleFX)
              (CallFunction|CameraShakeonImpact))))) ; client functionality, fx, sounds, camera shake
    )))

(fn CameraShakeOnImpact
  (for _array_element (Collision|GetOverlappingActors (Variables|Default|GetBox) "/Script/FortniteGame.FortPlayerPawnAthena" :!class /Script/Engine.PrimitiveComponent)
    (Game|Feedback|ClientStartCameraShake :self (Fortnite|GetFortPlayerControllerfromActor _array_element) :Shake "/Game/Environments/Helios/Sets/ScrapHeap/Blueprints/Engine/CameraShake/B_PistonCameraShake.B_PistonCameraShake_C" :Scale 1.0 :PlaySpace "CameraLocal")))

(fn ResetMeshWPO
  (Rendering|Material|SetScalarParameterforCustomPrimitiveData (Variables|Debug|GetStaticMeshComponent) "LoopDuration" (Variables|Default|GetLoopDuration))
  (Rendering|Material|SetScalarParameterforCustomPrimitiveData (Variables|Debug|GetStaticMeshComponent) "GameTime" (Utilities|Time|GetGameTimeinSeconds)))

(fn SpawnParticleFX
  (Niagara|SpawnSystematLocation "/Game/Environments/Helios/FX/OpenDesert/FX/NS_NitroTown_Pistons_Flames_Sparks_OnePiston.NS_NitroTown_Pistons_Flames_Sparks_OnePiston" (.location (Transformation|GetSocketTransform (Variables|Debug|GetStaticMeshComponent) "FXImpact" "RTS_World")) (.rotation (Transformation|GetSocketTransform (Variables|Debug|GetStaticMeshComponent) "FXImpact" "RTS_World"))))

(fn StartImpactSound
  (Audio|Components|Audio|Play (Variables|Default|GetImpactSound) 0.0))

(fn StartMovingSound
  (Audio|Components|Audio|Play (Variables|Default|GetMovementSound) 0.0))

(fn ConstructionScript)
